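# Niagara idle effect for the Daytrader character body that flickers small bright sprite glints on two skeletal meshes.
# Glint_L and Glint_R are near-identical emitters, each sampling random bone or socket indices on its own user-supplied mesh during particle update.
# Randomized loops burst one guaranteed glint plus probabilistic extras, which curves fade and scale in and out with a camera offset.
# Unlike Glint_L, Glint_R disables killing particles when lifetime elapses.
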NiagaraSystem /BRCosmetics/Characters/Player/Male/Medium/Bodies/M_MED_Daytrader/FX/NS_DayTrader_Idle
dependencies {usedDataInterfaces [&NiagaraDataInterfaceCurve], usedDynamicInputs [], usedModules [&SystemState]}
emitters {
 Glint_L {
  emitterUpdateScript [
   &EmitterState(
    Life Cycle Mode=::NewEnumerator1
    Loop Delay=&UniformRangedFloat(Minimum=0, Maximum=1)
    Loop Duration=&UniformRangedFloat(Minimum=0.5, Maximum=1)
    Recalculate Duration Each Loop=true
    UseLoopDelay=true
   )
   &SpawnBurst_Instantaneous(Spawn Count=1, Use Spawn Probability=false)
   &SpawnBurst_Instantaneous(Spawn Count=1, Spawn Probability=0.4, Use Spawn Probability=true) {module=SpawnBurst_Instantaneous001}
   &SpawnBurst_Instantaneous(Spawn Count=1, Spawn Probability=0.1, Use Spawn Probability=true) {module=SpawnBurst_Instantaneous002}
  ]
  particleSpawnScript [/Niagara/Modules/Spawn/Initialization/InitializeParticle(
    Color=<rgba 5 5 5 1>
    Lifetime=&UniformRangedFloat(Minimum=0.7, Maximum=1)
    Module.Write Color=true
    Module.Write Lifetime=true
    Module.Write Mass=true
    Module.Write Material Random=true
    Module.Write Position=true
    Module.Write Scale=false
    Module.Write SpriteRotation=true
    Module.Write SpriteSize=true
    Sprite Rotation=&UniformRangedFloat(Minimum=0, Maximum=360)
    Sprite Size=&Vector2DFromFloat(Value=&UniformRangedFloat(Minimum=8, Maximum=10))
   )]
  particleUpdateScript [
   &SkeletalMeshLocation(
    Bone / Socket Index=&RandomRangeInt(Minimum=0, Maximum=16)
    Inherit Velocity=false
    "Inherit Velocity (Scale)"=20
    MaskSpawnGroup=false
    Mesh Sampling Type=::NewEnumerator5
    Sampled Position Offset Coordinate Space:ENiagaraMeshTransforms=::NewEnumerator3
    Skeletal Mesh=$User.SkeletalMesh
    Socket Sampling Mode=::NewEnumerator5
    Write Sampled Mesh Attrributes to Particle Attributes=true
   )
   &ScaleColor(
    Scale Alpha=&FloatFromCurve(FloatCurve=DataInterface() {propertyValues {Curve {keys [<at 0 0>,<at 0.5016 0.9767>,<at 1 0>]}}})
    Scale RGB=&VectorFromFloat(Value=&FloatFromCurve(FloatCurve=DataInterface() {propertyValues {Curve {keys [<at 0 0>,<at 0.5017 1.0042>,<at 1 0>]}}}))
    ScaleA=true
    ScaleRGB=true
   )
   &ParticleState()
   &ScaleSpriteSize(Scale Factor=&Vector2DFromFloat(Value=&FloatFromCurve(
      FloatCurve=DataInterface() {propertyValues {Curve {keys [
          <at 0 0 RCIM_Cubic>
          {arriveTangent=1.9892, interpMode=RCIM_Cubic, leaveTangent=1.9892, time=0.335, value=0.2056}
          {arriveTangent=0.2473, interpMode=RCIM_Cubic, leaveTangent=0.2473, time=0.5002, value=0.995}
          {arriveTangent=-1.9908, interpMode=RCIM_Cubic, leaveTangent=-1.9908, time=0.6669, value=0.2877}
          <at 1 0>
         ]}}}
      Scale Curve=1.5
     )))
   &CameraOffset(Camera Offset Amount=2.5)
   &SolveForcesAndVelocity(Clamp Velocity=false, Limit Acceleration=false)
  ]
  rendererClasses [&NiagaraSpriteRendererProperties]
  renderers [{rendererClass=&NiagaraSpriteRendererProperties}]
 }
 Glint_R {
  emitterUpdateScript [
   &EmitterState(
    Life Cycle Mode=::NewEnumerator1
    Loop Delay=&UniformRangedFloat(Minimum=0, Maximum=1)
    Loop Duration=&UniformRangedFloat(Minimum=0.5, Maximum=1)
    Recalculate Duration Each Loop=true
    UseLoopDelay=true
   )
   &SpawnBurst_Instantaneous(Spawn Count=1, Use Spawn Probability=false)
   &SpawnBurst_Instantaneous(Spawn Count=1, Spawn Probability=0.4, Use Spawn Probability=true) {module=SpawnBurst_Instantaneous001}
   &SpawnBurst_Instantaneous(Spawn Count=1, Spawn Probability=0.1, Use Spawn Probability=true) {module=SpawnBurst_Instantaneous002}
  ]
  particleSpawnScript [&InitializeParticle(
    Color=<rgba 5 5 5 1>
    Lifetime=&UniformRangedFloat(Minimum=0.7, Maximum=1)
    Module.Write Color=true
    Module.Write Lifetime=true
    Module.Write Mass=true
    Module.Write Material Random=true
    Module.Write Position=true
    Module.Write Scale=false
    Module.Write SpriteRotation=true
    Module.Write SpriteSize=true
    Sprite Rotation=&UniformRangedFloat(Minimum=0, Maximum=360)
    Sprite Size=&Vector2DFromFloat(Value=&UniformRangedFloat(Minimum=8, Maximum=10))
   )]
  particleUpdateScript [
   &SkeletalMeshLocation(
    Bone / Socket Index=&RandomRangeInt(Minimum=0, Maximum=16)
    Inherit Velocity=false
    "Inherit Velocity (Scale)"=20
    MaskSpawnGroup=false
    Mesh Sampling Type=::NewEnumerator5
    Sampled Position Offset Coordinate Space:ENiagaraMeshTransforms=::NewEnumerator3
    Skeletal Mesh=$User.SkeletalMesh_02
    Socket Sampling Mode=::NewEnumerator5
    Write Sampled Mesh Attrributes to Particle Attributes=true
   )
   &ScaleColor(
    Scale Alpha=&FloatFromCurve(FloatCurve=DataInterface() {propertyValues {Curve {keys [<at 0 0>,<at 0.5016 0.9767>,<at 1 0>]}}})
    Scale RGB=&VectorFromFloat(Value=&FloatFromCurve(FloatCurve=DataInterface() {propertyValues {Curve {keys [<at 0 0>,<at 0.5017 1.0042>,<at 1 0>]}}}))
    ScaleA=true
    ScaleRGB=true
   )
   &ParticleState(Kill Particles When Lifetime Has Elapsed=false)
   &ScaleSpriteSize(Scale Factor=&Vector2DFromFloat(Value=&FloatFromCurve(
      FloatCurve=DataInterface() {propertyValues {Curve {keys [
          <at 0 0 RCIM_Cubic>
          {arriveTangent=1.9892, interpMode=RCIM_Cubic, leaveTangent=1.9892, time=0.335, value=0.2056}
          {arriveTangent=0.2473, interpMode=RCIM_Cubic, leaveTangent=0.2473, time=0.5002, value=0.995}
          {arriveTangent=-1.9908, interpMode=RCIM_Cubic, leaveTangent=-1.9908, time=0.6669, value=0.2877}
          <at 1 0>
         ]}}}
      Scale Curve=1.5
     )))
   &CameraOffset(Camera Offset Amount=2.5)
   &SolveForcesAndVelocity(Clamp Velocity=false, Limit Acceleration=false)
  ]
  rendererClasses [&NiagaraSpriteRendererProperties]
  renderers [{rendererClass=&NiagaraSpriteRendererProperties}]
 }
}
summary {
 systemName=NS_DayTrader_Idle
 emitters [
  {emitterName=Glint_L, rendererClasses [&NiagaraSpriteRendererProperties]}
  {emitterName=Glint_R, rendererClasses [&NiagaraSpriteRendererProperties]}
 ]
 userVariables [
  User.SkeletalMesh {
   type=NiagaraDataInterfaceSkeletalMesh
   defaultValue {
    struct=NiagaraExt_VariableValue_DataInterface
    value {dataInterface=&self:NiagaraDataInterfaceSkeletalMesh_0}
   }
  }
  User.SkeletalMesh_02 {
   type=NiagaraDataInterfaceSkeletalMesh
   defaultValue {
    struct=NiagaraExt_VariableValue_DataInterface
    value {dataInterface=&self:NiagaraDataInterfaceSkeletalMesh_1}
   }
  }
 ]
}
