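# Parent character material for the Daytrader skin, rendering a faceted diamond look on an opaque lit surface.
# A facet texture serves as normals for cubemap lookups that fake refraction, with sphere-masked, hue-shifted curve-atlas "laser" color.
# A view-offset facet sample, Fresnel on a detail-blended normal and sunlight-weighted highlights build the emissive, dimmed by time of day.
# Shared character PreFX and PostFX functions wrap the result.
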Material /BRCosmetics/Characters/Player/Male/Medium/Bodies/M_MED_Daytrader/Materials/MAT_DT_Parent
graph {
 properties {
  bUseMaterialAttributes=true
  bUsedWithMeshDeformer=true
  bUsedWithMorphTargets=true
  bUsedWithSkeletalMesh=true
  pixelDepthOffsetMode=PDOM_Legacy
  usedShadingModels=Default Lit
 }
 nodes [
  n1 = TextureCoordinate() {coordinateIndex=1}
  n2 = TextureSampleParameter2D(UVs=@n1) {
   parameterName=DT_FacetTexture
   texture=/BRCosmetics/Characters/Player/Male/Medium/Bodies/M_MED_Daytrader/Textures/T_M_MED_Daytrader_HandFacets
  }
  n3 = Multiply(A=@n2.RGB) {constB=0.6}
  n4 = TextureSampleParameterCube!(UVs=Transform(None=ReflectionVectorWS(CustomWorldNormal=@n3) {bNormalizeCustomWorldNormal=true}), Apply View MipBias) {
   group=DayTrader_Diamond
   parameterName=DT_RefractionCubeMap
   sortPriority=71
   texture=/Game/Packages/Fortress_Sky/TexturesHDR/T_Cubemap_Skylight_Subtle
  }
  n5 = ScalarParameter() {defaultValue=0.4, group=DayTrader_Diamond, parameterName=Refraction Intensity, sortPriority=71}
  n6 = MaterialFunctionCall(Float3=@n2.RGB) {materialFunction=&BreakOutFloat3Components}
  n7 = ReflectionVectorWS(CustomWorldNormal=AppendVector(A=AppendVector(A=@n6.B, B=@n6.B), B=@n6.G)) {bNormalizeCustomWorldNormal=true}
  n8 = TextureSampleParameterCube(UVs=Transform(None=Multiply(A=@n7) {constB=-1})) {
   group=DayTrader_Diamond
   parameterName=DT_RefractionCubeMap
   sortPriority=71
   texture=&T_Cubemap_Skylight_Subtle
  }
  n9 = ScalarParameter() {defaultValue=0.4, group=DayTrader_Diamond, parameterName=Refraction Intensity, sortPriority=71}
  n10 = ReflectionVectorWS(CustomWorldNormal=ConstantBiasScale(None=@n2.R) {bias=-0.5, scale=2})
  n11 = MaterialFunctionCall(Float3=Normalize(VectorInput=Abs(None=@n10))) {materialFunction=&BreakOutFloat3Components}
  n12 = LinearInterpolate(A=Multiply(A=Multiply(A=@n4.RGB, B=@n5), B=Multiply(A=@n8.RGB, B=@n9)), B=Constant() {r=0.05}, Alpha=Multiply(A=@n11.G))
  n13 = SphereMask(
   A=@n12
   B=ScalarParameter() {defaultValue=0.18, group=DayTrader_Diamond, parameterName=DT_LasorColorThreshold, sortPriority=20}
   Radius=ScalarParameter() {defaultValue=0.1, group=DayTrader_Diamond, parameterName=DT_LasorColorRange, sortPriority=21}
   Hardness=Constant()
  )
  n14 = CurveAtlasRowParameter(CurveTime=@n13) {
   atlas=&AthenaCurveAtlas01
   curve=/Game/Athena/Environments/Nexus/Materials/FXshader/Curve_Aix
   defaultValue=77
   group=DayTrader_Diamond
   parameterName=DT_LasorColorCurve
   sortPriority=25
  }
  n15 = Multiply(
   A=MaterialFunctionCall(
    Hue Shift Percentage=ScalarParameter() {group=DayTrader_Diamond, parameterName=DT_LasorHueShiftPCT, sortPriority=22}
    Texture=Multiply(A=@n14, B=@n13)
   ) {materialFunction=&HueShift}.Result
   B=ScalarParameter() {defaultValue=1, group=DayTrader_Diamond, parameterName=DT_LasorStrength, sortPriority=23}
  )
  n16 = Subtract(A=CameraPositionWS(), B=WorldPosition().XYZ)
  n17 = Normalize(VectorInput=ComponentMask(None=@n16) {g=true, r=true})
  n18 = TextureSampleParameter2D(UVs=Add(A=@n1, B=Multiply(A=@n17) {constB=0.02})) {
   parameterName=DT_FacetTexture
   texture=&T_M_MED_Daytrader_HandFacets
  }
  n19 = MaterialFunctionCall(Float3=@n18.RGB) {materialFunction=&BreakOutFloat3Components}
  n20 = AppendVector(A=AppendVector(A=@n19.B, B=@n19.B), B=@n19.G)
  n21 = Abs(None=Normalize(VectorInput=@n16))
  n22 = ConstantBiasScale(None=DotProduct(A=Multiply(A=@n20) {constB=0.6}, B=@n21)) {bias=0, scale=0.3}
  n23 = LinearInterpolate(A=@n22, B=Multiply(A=@n22))
  n24 = Add(A=@n12, B=ScalarParameter() {defaultValue=0.01, group=DayTrader_Diamond, parameterName=DT_DarkAreaEmissiveBoost})
  n25 = Add(A=@n15, B=@n24)
  n26 = DotProduct(A=@n3, B=@n21)
  n27 = ScalarParameter() {defaultValue=2, group=DayTrader_Diamond, parameterName=DT_HighlightAmountExp}
  n28 = Power(Base=SphereMask(
    A=Power(Base=@n26, Exp=@n27) {constExponent=1}
    B=Constant() {r=0.7}
    Radius=ScalarParameter() {defaultValue=0.2, group=DayTrader_Diamond, parameterName=DT_HighlightAmountRadius}
    Hardness=ScalarParameter() {defaultValue=0.5, group=DayTrader_Diamond, parameterName=DT_HighlightGradientHardness}
   ))
  n29 = Saturate(None=Add(A=@n28, B=ScalarParameter() {defaultValue=0.5, group=DayTrader_Diamond, parameterName=DT_SurfaceOpacity}) {constB=0.3})
  n30 = VectorParameter() {group=DayTrader_Diamond, parameterName=DT_EMSColorTint, sortPriority=15, defaultValue=<rgba 0.841 0.8376 0.9688 1>}
  n31 = Multiply(
   A=TextureCoordinate() {coordinateIndex=1}
   B=ScalarParameter() {defaultValue=4, group=DayTrader_Diamond, parameterName=DT_AdditionalNormalTiling, sortPriority=60}
  ) {constB=6}
  n32 = MaterialFunctionCall() {materialFunction=&BaseFN_CharacterFunction_PreFX}
  n33 = GetMaterialAttributes(None=@n32.Result) {attributeGetTypes [#1]}
  n34 = MaterialFunctionCall(
   BaseNormal=@n33.Normal
   AdditionalNormal=LinearInterpolate(
    A=Constant3Vector() {constant=<rgba 0 0 1 0>}
    B=TextureSample(UVs=@n31) {samplerType=SAMPLERTYPE_Normal, texture=/Game/Weapons/WeaponSkins/Wraps/Textures/AlienRock/T_Wrap_AlienRock_n}.RGB
    Alpha=ScalarParameter() {defaultValue=0.5, group=DayTrader_Diamond, parameterName=DT_AdditionalNormalIntensity, sortPriority=62}
   ) {constAlpha=1}
  ) {materialFunction=&BlendAngleCorrectedNormals}
  n35 = Fresnel(
   ExponentIn=ScalarParameter() {defaultValue=2, group=DayTrader_Diamond, parameterName=DT_FresnelExp, sortPriority=50}
   Normal=Transform(None=@n34.Result)
  )
  n36 = Fresnel()
  n37 = VectorParameter() {group=DayTrader_Diamond, parameterName=DT_FresnelColor, sortPriority=50, defaultValue=<rgba 1 1 1 1>}
  n38 = Add(A=Multiply(A=LinearInterpolate(A=@n23, B=@n25, Alpha=@n29), B=@n30.RGB), B=Multiply(A=LinearInterpolate(B=@n35, Alpha=@n36), B=@n37.RGB))
  n39 = VectorParameter() {group=DayTrader_Diamond, parameterName=DT_HighlightColor, sortPriority=12, defaultValue=<rgba 1 1 1 1>}
  n40 = Multiply(A=Saturate(None=Multiply(A=@n11.R, B=@n28)), B=@n39.RGB) {constB=5}
  n41 = ScalarParameter() {defaultValue=4, group=DayTrader_Diamond, parameterName=DT_HighlightMult, sortPriority=11}
  n42 = CollectionParameter() {
   collection=&FortniteMaterialParameters
   parameterName=SunlightDirectionalVector
  }
  n43 = DotProduct(A=ComponentMask(None=@n42) {b=true, g=true, r=true}, B=ComponentMask(None=VertexNormalWS()) {b=true, g=true, r=true})
  n44 = LinearInterpolate(A=@n38, B=Add(A=Multiply(A=@n40, B=@n41) {constB=5}, B=@n38), Alpha=Clamp(None=Saturate(None=@n43)) {minDefault=0.1})
  Add(A=@n15, B=@n44)
  n45 = VectorParameter() {group=DayTrader_Diamond, parameterName=DT_BaseColorTint, sortPriority=15, defaultValue=<rgba 1 1 1 1>}
  n46 = Saturate(None=@n24)
  n47 = Constant() {r=1}
  n48 = Constant() {r=0.5}
  n49 = SetMaterialAttributes!(
   MaterialAttributes=@n33.MaterialAttributes
   Base Color=Multiply(A=@n45.RGB, B=Clamp(None=@n46) {maxDefault=0.5})
   Metallic=ScalarParameter() {group=DayTrader_Diamond, parameterName=DT_Metallic, sortPriority=36}
   Specular=ScalarParameter() {group=DayTrader_Diamond, parameterName=DT_Specular, sortPriority=34}
   Roughness=ScalarParameter() {group=DayTrader_Diamond, parameterName=DT_Roughness, sortPriority=35}
   Emissive Color=Multiply(
    A=MaterialFunctionCall(Value_Morning=@n47, Value_Day=@n47, Value_Evening=@n47, Value_Night=@n48) {materialFunction=&Scalar_Time_of_Day}.Result
    B=@n44
   )
   Normal=@n34.Result
  ) {attributeSetTypes [
    #2
    #3
    #4
    #5
    #6
    #1
   ]}
  MaterialFunctionCall(AO=@n32.AO, Distance Blend=@n32.AO, Fresnel=@n32.AO, Gmap=@n32.AO, Pre Skinned Local Position=@n32.AO, Material=@n49) {materialFunction=&BaseFN_CharacterFunction_PostFX}
  Reroute(None=@n12)
 ]
}
